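; Blueprint for a hanging lamp prop variant C in the Asteria Amber Forest Town environment set.
; Its construction script calls MakeConfigurable on a Lumen emissive light component, using the component obtained through a Lumen emissive light getter.
; The file shows no other logic.
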(fn ConstructionScript
  (Class|LumenEmissiveLightComponent|MakeConfigurable (Variables|Default|GetLumenEmissiveLight_)))
